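# Post-process-style Anim Blueprint for the male BrakePedal character body, adding deform and accessory bone motion on top of the pose copied from its attached parent mesh.
# A deform Control Rig feeds two BrakePedal Control Rig instances, zipline on and off, blended by OnZipline, then weighted multi-bone constraints drive glute and shoulder base bones.
# The EventGraph turns skydive and parachute states into 0/1 rig inputs, and zipline enter/exit anim input events toggle OnZipline.
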AnimBlueprint /BRCosmetics/Characters/Player/Male/Medium/Bodies/M_MED_BrakePedal/Meshes/M_MED_BrakePedal_AnimBP
target_skeleton=&/M_MED_BrakePedal_Skeleton
graphs [
 AnimGraph {nodes [
   Output Pose(Result=Use cached pose 'CR_Constraints'()) {layerGroup=None}
   n1 = Get SkyDiving() -> (SkyDiving:double)
   n2 = Get SkyDiving_inverse() -> (SkyDiving_inverse:double)
   n3 = Get ParachuteOpen() -> (ParachuteOpen:double)
   n4 = Control Rig!(
    Source=Use cached pose 'CR_DeformRig'()
    Alpha
    SkyDiving:double=@n1
    SkyDive_Inverse:double=@n2
    ParachuteOpen:double=@n3
    OnZipline:double=0
   ) -> (Pose) {
    alphaBoolBlend {blendOption=Linear}
    controlRigAssetReference {blueprintRigClass=&/M_MED_BrakePedal_CtrlRig.M_MED_BrakePedal_CtrlRig_C}
   }
   n5 = Control Rig(Source=Use cached pose 'CR_DeformRig'(), SkyDiving=@n1, SkyDive_Inverse=@n2, ParachuteOpen=@n3, OnZipline=1) {
    alphaBoolBlend {blendOption=Linear}
    controlRigAssetReference {blueprintRigClass=&M_MED_BrakePedal_CtrlRig_C}
   }
   Save cached pose 'CR_ControlRig'(Pose=Blend Poses by bool(BlendPose_0=@n5, BlendPose_1=@n4, BlendTime_0=0.3, BlendTime_1=0.3, bActiveValue=Get OnZipline() -> (OnZipline:bool)) {blendType=Linear}) {cacheName=CR_ControlRig}
   n6 = "Constraint - base_glute_wood_l (PPPP)"!(
    ComponentPose=Local To Component(LocalPose=Use cached pose 'CR_ControlRig'())
    Alpha=1
    ConstraintWeights_0=0.4447
    ConstraintWeights_1:float=0.4089
    ConstraintWeights_2:float=0.1401
    ConstraintWeights_3:float=0.0063
   ) -> (Pose) {
    alphaBoolBlend {blendOption=Linear}
    boneToModify {boneName=base_glute_wood_l}
    constraintSetup [
     {targetBone {boneName=deform_glute_l}}
     {targetBone {boneName=pelvis}}
     {targetBone {boneName=thigh_twist_01_l}}
     {targetBone {boneName=thigh_twist_01_r}}
    ]
    constraintWeights [1,1,1,1]
   }
   n7 = "Constraint - base_glute_wood_r (PPPP)"!(
    ComponentPose=@n6
    Alpha=1
    ConstraintWeights_0=0.4447
    ConstraintWeights_1:float=0.4089
    ConstraintWeights_2:float=0.1401
    ConstraintWeights_3:float=0.0063
   ) -> (Pose) {
    alphaBoolBlend {blendOption=Linear}
    boneToModify {boneName=base_glute_wood_r}
    constraintSetup [
     {targetBone {boneName=deform_glute_r}}
     {targetBone {boneName=pelvis}}
     {targetBone {boneName=thigh_twist_01_r}}
     {targetBone {boneName=thigh_twist_01_l}}
    ]
    constraintWeights [1,1,1,1]
   }
   n8 = "Constraint - base_shoulder_l (PPPPPPPP)"!(
    ComponentPose=@n7
    Alpha=1
    ConstraintWeights_0=0.442
    ConstraintWeights_1:float=0.237
    ConstraintWeights_2:float=0.1477
    ConstraintWeights_3:float=0.0908
    ConstraintWeights_4:float=0.0463
    ConstraintWeights_5:float=0.0185
    ConstraintWeights_6:float=0.0091
    ConstraintWeights_7:float=0.0086
   ) -> (Pose) {
    alphaBoolBlend {blendOption=Linear}
    boneToModify {boneName=base_shoulder_l}
    constraintSetup [
     {targetBone {boneName=upperarm_twist_01_l}}
     {targetBone {boneName=deform_upperarm_up_l}}
     {targetBone {boneName=upperarm_twist_02_l}}
     {targetBone {boneName=clavicle_l}}
     {targetBone {boneName=upperarm_l}}
     {targetBone {boneName=deform_upperarm_fr_l}}
     {targetBone {boneName=spine_05}}
     {targetBone {boneName=spine_03}}
    ]
    constraintWeights [1,1,1,1,1,1,1,1]
   }
   n9 = "Constraint - base_shoulder_r (PPPPPPPP)"!(
    ComponentPose=@n8
    Alpha=1
    ConstraintWeights_0=0.442
    ConstraintWeights_1:float=0.237
    ConstraintWeights_2:float=0.1477
    ConstraintWeights_3:float=0.0908
    ConstraintWeights_4:float=0.0463
    ConstraintWeights_5:float=0.0185
    ConstraintWeights_6:float=0.0091
    ConstraintWeights_7:float=0.0086
   ) -> (Pose) {
    alphaBoolBlend {blendOption=Linear}
    boneToModify {boneName=base_shoulder_r}
    constraintSetup [
     {targetBone {boneName=upperarm_twist_01_r}}
     {targetBone {boneName=deform_upperarm_up_r}}
     {targetBone {boneName=upperarm_twist_02_r}}
     {targetBone {boneName=clavicle_r}}
     {targetBone {boneName=upperarm_r}}
     {targetBone {boneName=deform_upperarm_fr_base_r}}
     {targetBone {boneName=spine_05}}
     {targetBone {boneName=spine_03}}
    ]
    constraintWeights [1,1,1,1,1,1,1,1]
   }
   Save cached pose 'CR_Constraints'(Pose=Component To Local(ComponentPose=@n9)) {cacheName=CR_Constraints}
   n10 = Control Rig!(Source=Copy Pose From Mesh() {bUseAttachedParent=true}, Alpha) -> (Pose) {
    alphaBoolBlend {blendOption=Linear}
    controlRigAssetReference {blueprintRigClass=&/M_MED_DeformRig_BrakePedal_ControlRig.M_MED_DeformRig_BrakePedal_ControlRig_C}
   }
   Save cached pose 'CR_DeformRig'(Pose=@n10) {cacheName=CR_DeformRig}
  ]}
 EventGraph {nodes [
   Event BlueprintUpdateAnimation()
   > n1 = Sequence!(execute) -> (then_0, then_1, then_2:Exec, then_3:Exec)
   n2 = Branch(execute=@n1.then_1, Condition=Get bIsParachuteOpened())
   > Set ParachuteOpen(ParachuteOpen:double=1) -> (Output_Get:double)
   Set ParachuteOpen(execute=@n2.else, ParachuteOpen=0)
   n3 = Branch(execute=@n1.then_0, Condition=OR Boolean!(A=Get bIsSkyDiving(), B=false, C:bool=Get bIsSkydiveFloating(), D:bool) -> (ReturnValue))
   > Set SkyDiving(SkyDiving:double=1) -> (Output_Get:double)
   > Set SkyDiving_inverse(SkyDiving_inverse:double=0) -> (Output_Get:double)
   Set SkyDiving(execute=@n3.else, SkyDiving=0)
   > Set SkyDiving_inverse(SkyDiving_inverse=1)
   n4 = Event BlueprintProcessAnimInputEvent()
   > n5 = Branch(Condition=IsNamed(self=@n4.AnimInputEvent, InEventName=aie.Player.Zipline.Entered))
   > Set OnZipline(OnZipline:bool=1) -> (Output_Get:bool)
   Branch(execute=@n5.else, Condition=IsNamed(self=@n4.AnimInputEvent, InEventName=aie.Player.Zipline.Exited))
   > Set OnZipline(OnZipline=false)
   Get bIsSkydiveDiveMode()
  ]}
]
variables [
 OnZipline {type=boolean}
 ParachuteOpen {type=number}
 SkyDiving {type=number}
 SkyDiving_inverse {type=number}
 Spine_03_rot_back {type=number}
 Spine_03_rot_front {type=number}
]
